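# UI-domain material for the Lined Notebook Item Shop tile, composing a green-toned background.
# Nested lerps layer a green-tinted brick texture over black, add a smoothstepped lime glow, and blend against a dark teal base.
# Three offset RadialGradientExponential masks, published as named reroutes (Purple, Green, Shading), sit alongside a UVScaling_Y reroute from MF_UI_UVScaling.
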Material /ItemShop/UI/Art/LinedNotebook/M_UI_LinedNotebook_ShopTile
graph {
 properties {bUsedWithStaticMesh=false, materialDomain=MD_UI, pixelDepthOffsetMode=PDOM_Legacy, usedShadingModels=Default Lit}
 nodes [
  n1 = Constant3Vector() {constant=<rgba 0 0 0 1>}
  n2 = Multiply(A=NamedRerouteUsage()) {constB=0.75}
  n3 = Multiply(
   A=Divide(A=TextureSample(UVs=@n2) {texture=/OfferCatalog/Art/A_Shop_Tiles_Textures/T_UI_LNB_Bricks}.R) {constB=0.07}
   B=Constant3Vector() {constant=<rgba 0 0.1146 0.0098 1>}
  )
  n4 = NamedRerouteUsage()
  n5 = NamedRerouteUsage()
  n6 = Multiply(A=Constant3Vector() {constant=<rgba 0.7967 1 0 1>}, B=@n5)
  n7 = SmoothStep(Value=@n5) {constMax=2.2}
  LinearInterpolate(
   A=Constant3Vector() {constant=<rgba 0.0063 0.0625 0.0477 1>}
   B=LinearInterpolate(A=LinearInterpolate(A=@n1, B=@n3, Alpha=@n4) {constB=0}, B=Divide(A=@n6) {constB=0.4}, Alpha=Divide(A=@n7) {constB=0.35})
   Alpha=Multiply(A=NamedRerouteUsage())
  )
  n8 = NamedRerouteUsage()
  n9 = AppendVector(A=Constant(), B=Constant() {r=0.5})
  NamedRerouteDeclaration(None=MaterialFunctionCall(UVs=Add(A=@n8, B=@n9), Radius=Constant() {r=1.25}, Density=Constant() {r=2}) {materialFunction=&RadialGradientExponential}.RadialGradientExponential) {name=Purple Gradient MAsk, nodeColor=<rgba 0.3392 0 1 1>}
  n10 = Add(A=NamedRerouteUsage(), B=AppendVector(A=Constant(), B=Constant() {r=0.2}))
  n11 = Constant() {r=1}
  n12 = Constant() {r=0.7}
  NamedRerouteDeclaration(None=SmoothStep(Value=MaterialFunctionCall(UVs=@n10, Radius=@n11, Density=@n12) {materialFunction=&RadialGradientExponential}.RadialGradientExponential) {constMax=0.6}) {name="Shading Gradient  Mask", nodeColor=<rgba 0.0131 0.0278 0 1>}
  NamedRerouteDeclaration(None=MaterialFunctionCall() {materialFunction=&MF_UI_UVScaling}.ScaleY) {name=UVScaling_Y, nodeColor=<rgba 0 0.3203 1 1>}
  n13 = NamedRerouteUsage()
  n14 = AppendVector(A=Constant(), B=Constant() {r=0.5})
  NamedRerouteDeclaration(None=MaterialFunctionCall(UVs=Add(A=@n13, B=@n14), Radius=Constant() {r=1.25}, Density=Constant() {r=0.75}) {materialFunction=&RadialGradientExponential}.RadialGradientExponential) {name="Green Gradient  Mask", nodeColor=<rgba 0.4726 1 0 1>}
 ]
}
